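# Sound cue for looping government computer screen ambience, attenuated over a short range and limited to two concurrent instances.
# A mixer layers a pitched-down transmitter running loop with three looping branches of randomly chosen computer one-shots, each separated by random delays and pitch-modulated.
# A top-level random node gives that mix even odds against an empty branch, so some instances stay silent.
# Only the first two quality-level slots are connected; the other two are empty.
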SoundCue /Game/Sounds/Artemis/Environment_Actors/Computer_Screen/Government_Computer_Loop_Cue
root=QualityLevel_0
nodes [
 {id=Delay_0, children [Modulator_1], properties {delayMax=1, delayMin=0.5}}
 {id=Delay_1, children [Modulator_0], properties {delayMax=1, delayMin=0.5}}
 {id=Delay_2, children [Modulator_4], properties {delayMax=1, delayMin=0.5}}
 {id=Looping_0, children [Delay_0]}
 {id=Looping_1, children [Delay_1]}
 {id=Looping_2, children [Delay_2]}
 {id=Mixer_0, children [WavePlayer_9], properties {inputVolume [1.3]}}
 {id=Mixer_5, children [Looping_0,Looping_1,Looping_2,Modulator_6], properties {inputVolume [0.5,1,0.8,3]}}
 {id=Modulator_0, children [Random_2], properties {pitchMax=1.15, volumeMax=1, volumeMin=1}}
 {id=Modulator_1, children [Random_0], properties {pitchMax=1.15, volumeMax=1, volumeMin=1}}
 {id=Modulator_4, children [Random_5], properties {pitchMax=1.3, pitchMin=0.7, volumeMax=1, volumeMin=1}}
 {id=Modulator_6, children [WavePlayer_0], properties {pitchMax=0.8, pitchMin=0.5}}
 {id=QualityLevel_0, children [Random_1,Random_1,null,null]}
 {id=Random_0, children [WavePlayer_5,WavePlayer_6,WavePlayer_7,WavePlayer_8], properties {weights [1,1,1,1]}}
 {id=Random_1, children [Mixer_5,null], properties {weights [1,1]}}
 {id=Random_2, children [WavePlayer_4, Mixer_0, WavePlayer_10], properties {weights [1,1,1]}}
 {id=Random_5, children [WavePlayer_1, WavePlayer_2, WavePlayer_3, WavePlayer_12, WavePlayer_14], properties {weights [1,1,1,1,1]}}
 {
  id=WavePlayer_0
  properties {bLooping=true, soundWaveAssetPtr=/Game/Athena/Sounds/Props/MountainBaseComputer/ComputerTransimtter_Running_Loop_01}
 }
 {id=WavePlayer_1, properties {soundWaveAssetPtr=/Game/Sounds/Apollo/Biomes/GovernmentSecretComplex/SecretComplex_Computer_Oneshot_10}}
 {id=WavePlayer_10, properties {soundWaveAssetPtr=/Game/Sounds/Apollo/Biomes/GovernmentSecretComplex/SecretComplex_Computer_Oneshot_07}}
 {id=WavePlayer_12, properties {soundWaveAssetPtr=/Game/Sounds/Apollo/Biomes/GovernmentSecretComplex/SecretComplex_Computer_Oneshot_13}}
 {id=WavePlayer_14, properties {soundWaveAssetPtr=/Game/Sounds/Apollo/Biomes/GovernmentSecretComplex/SecretComplex_Computer_Oneshot_15}}
 {id=WavePlayer_2, properties {soundWaveAssetPtr=/Game/Sounds/Apollo/Biomes/GovernmentSecretComplex/SecretComplex_Computer_Oneshot_11}}
 {id=WavePlayer_3, properties {soundWaveAssetPtr=/Game/Sounds/Apollo/Biomes/GovernmentSecretComplex/SecretComplex_Computer_Oneshot_12}}
 {id=WavePlayer_4, properties {soundWaveAssetPtr=/Game/Sounds/Apollo/Biomes/GovernmentSecretComplex/SecretComplex_Computer_Oneshot_01}}
 {id=WavePlayer_5, properties {soundWaveAssetPtr=/Game/Sounds/Apollo/Biomes/GovernmentSecretComplex/SecretComplex_Computer_Oneshot_02}}
 {id=WavePlayer_6, properties {soundWaveAssetPtr=/Game/Sounds/Apollo/Biomes/GovernmentSecretComplex/SecretComplex_Computer_Oneshot_03}}
 {id=WavePlayer_7, properties {soundWaveAssetPtr=/Game/Sounds/Apollo/Biomes/GovernmentSecretComplex/SecretComplex_Computer_Oneshot_04}}
 {id=WavePlayer_8, properties {soundWaveAssetPtr=/Game/Sounds/Apollo/Biomes/GovernmentSecretComplex/SecretComplex_Computer_Oneshot_05}}
 {id=WavePlayer_9, properties {soundWaveAssetPtr=/Game/Sounds/Apollo/Biomes/GovernmentSecretComplex/SecretComplex_Computer_Oneshot_06}}
]
properties {
 attenuationSettings=&/Government_Computer_Att
 bOverrideConcurrency=true
 maxDistance=750
 soundClassObject=&Fort_Gadgets
 volumeMultiplier=0.4
 attenuationOverrides {
  bEnableFocusInterpolation=true
  falloffDistance=600
  nonFocusVolumeAttenuation=0.3
  occlusionVolumeAttenuation=0.3
  stereoSpread=0
  attenuationShapeExtents {x=200}
 }
 concurrencyOverrides {maxCount=2}
}
