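# Looping sound cue for a government computer screen hum, capped by a concurrency override.
# A mixer layers a low-pitched looping transmitter hum with two repeating layers of randomly picked computer one-shots.
# One one-shot layer repeats continuously and the other after a short random delay, both with pitch variation.
# Only the first two quality levels are wired to the mix; the other two are empty.
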SoundCue /Game/Sounds/Artemis/Environment_Actors/Computer_Screen/Government_Computer_Screen_Hum_Loop_Cue
root=QualityLevel_0
nodes [
 {id=Delay_1, children [Modulator_0], properties {delayMax=1, delayMin=0.5}}
 {id=Looping_0, children [Modulator_1]}
 {id=Looping_1, children [Delay_1]}
 {id=Mixer_5, children [Looping_0,Looping_1,Modulator_6], properties {inputVolume [0.4,0.2,3]}}
 {id=Modulator_0, children [Random_2], properties {pitchMax=1.15, volumeMax=1, volumeMin=1}}
 {id=Modulator_1, children [Random_0], properties {pitchMax=1.15, volumeMax=1, volumeMin=1}}
 {id=Modulator_6, children [WavePlayer_0], properties {pitchMax=0.9, pitchMin=0.7}}
 {id=QualityLevel_0, children [Random_1,Random_1,null,null]}
 {id=Random_0, children [WavePlayer_5,WavePlayer_6,WavePlayer_7,WavePlayer_8], properties {weights [1,1,1,1]}}
 {id=Random_1, children [Mixer_5], properties {weights [1]}}
 {id=Random_2, children [WavePlayer_4, WavePlayer_10], properties {weights [1,1]}}
 {
  id=WavePlayer_0
  properties {bLooping=true, soundWaveAssetPtr=/Game/Athena/Sounds/Props/MountainBaseComputer/ComputerTransimtter_Running_Loop_01}
 }
 {id=WavePlayer_10, properties {soundWaveAssetPtr=/Game/Sounds/Apollo/Biomes/GovernmentSecretComplex/SecretComplex_Computer_Oneshot_07}}
 {id=WavePlayer_4, properties {soundWaveAssetPtr=/Game/Sounds/Apollo/Biomes/GovernmentSecretComplex/SecretComplex_Computer_Oneshot_01}}
 {id=WavePlayer_5, properties {soundWaveAssetPtr=/Game/Sounds/Apollo/Biomes/GovernmentSecretComplex/SecretComplex_Computer_Oneshot_02}}
 {id=WavePlayer_6, properties {soundWaveAssetPtr=/Game/Sounds/Apollo/Biomes/GovernmentSecretComplex/SecretComplex_Computer_Oneshot_03}}
 {id=WavePlayer_7, properties {soundWaveAssetPtr=/Game/Sounds/Apollo/Biomes/GovernmentSecretComplex/SecretComplex_Computer_Oneshot_04}}
 {id=WavePlayer_8, properties {soundWaveAssetPtr=/Game/Sounds/Apollo/Biomes/GovernmentSecretComplex/SecretComplex_Computer_Oneshot_05}}
]
properties {
 attenuationSettings=&/Government_Computer_Smaller_Att
 bOverrideConcurrency=true
 maxDistance=700
 soundClassObject=&Fort_Gadgets
 volumeMultiplier=0.4
 attenuationOverrides {
  bEnableFocusInterpolation=true
  falloffDistance=600
  nonFocusVolumeAttenuation=0.3
  occlusionVolumeAttenuation=0.3
  stereoSpread=0
  attenuationShapeExtents {x=200}
 }
 concurrencyOverrides {maxCount=2}
}
